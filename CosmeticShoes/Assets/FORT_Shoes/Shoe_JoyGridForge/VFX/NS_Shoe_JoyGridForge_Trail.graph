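# Niagara trail effect for the JoyGridForge cosmetic shoes, combining foot ribbons, glinting sprites and mesh particles from an emitter named Bananas.
# Sprites and meshes spawn at skeletal mesh sockets while running and not skydiving, with mesh lifetime gated by owner speed and collision by platform quality.
# AnimTrail ribbons follow socket pairs, fade unless running, stop for skydiving or HideFootFX, and adjust for front end and time of day.
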NiagaraSystem /CosmeticShoes/Assets/FORT_Shoes/Shoe_JoyGridForge/VFX/NS_Shoe_JoyGridForge_Trail
dependencies {
 usedDataInterfaces [
  &NiagaraDataInterfaceCurve
  &NiagaraDataInterfacePlatformSet
  &NiagaraDataInterfaceColorCurve
  &NiagaraDataInterfaceSkeletalMesh
 ]
 usedDynamicInputs [&Boolean_OrOperation, /Game/Effects/Niagara/DynamicInputs/Logic/FN_IsRunningCheck]
 usedModules [
  &self:SystemSpawnScript.SystemScriptSource.SystemScriptGraph.NiagaraNodeAssignment_1.SetVariables_#1
  &SystemState
  &self:SystemSpawnScript.SystemScriptSource.SystemScriptGraph.NiagaraNodeAssignment_0.SetVariables_#2
 ]
}
emitters {
 Bananas {
  emitterUpdateScript [
   &EmitterState()
   &SpawnRate(
    SpawnRate=&MakeCustomFloatFromBool(
     Bool=&Boolean_AndOperation(Bool1=$System.bIsRunning, Bool2=&Boolean_NotOperation(Bool=$User.IsSkydiving))
     True Float=100
     False Float=0
    )
    Use Spawn Probability=false
   )
  ]
  particleSpawnScript [
   &PartitionParticles()
   &InitializeParticle(
    Color Mode=::NewEnumerator0
    Lifetime=&MakeCustomFloatFromBool(
     Bool=&SetBoolByFloatComparison(A=&VectorLength(Vector=$Engine.Owner.Velocity), B=800)
     True Float=0
     False Float=&UniformAOrBFloat(
      A=&RandomRangeFloat(Minimum=0.08, Maximum=0.2)
      Distribution Weight=0.12
      Evaluation Type=::NewEnumerator0
      Recalculate Random Each Loop=true
      Override Seed=false
     )
    )
    Mesh Scale Mode=::NewEnumerator1
    Mesh Uniform Scale Max=0.5
    Mesh Uniform Scale Min=0.3
    Sprite Size Mode=::NewEnumerator3
    Uniform Sprite Size=5
    UsePositionOffset=false
    Write Material Random=true
   )
   &SkeletalMeshLocation(
    Bone / Socket Index=$Particles.PartitionParticles.Partition
    MaskSpawnGroup=false
    Mesh Sampling Type=::NewEnumerator5
    Skeletal Mesh=$System.Skeletal Mesh
    Socket Sampling Mode=::NewEnumerator5
   )
   &ShapeLocation(MaskSpawnGroup=false, Non Uniform Scale=<1 0.2 1>, Override Seed=false, Sphere Radius=10, UseOwnerScale=true)
   &InitialMeshOrientation(Rotation=&RandomRangeVector(Minimum=<360 360 360>, Maximum=<-360 -360 -360>), Use Rotation Vector=true)
   &AddVelocity(
    Cone Axis=<0 -1 0>
    Module.UseConeFalloff=false
    Module.UseVelDistribution=false
    Override Seed=false
    Velocity Mode=::NewEnumerator2
    Velocity Speed=50
   )
  ]
  particleUpdateScript [
   &ParticleState()
   &UpdateMeshOrientation(Rotation Vector=<0 0 1>) {enabled=false}
   &GravityForce(Gravity=<0 0 -500>)
   &ScaleMeshSize(Scale Factor=&VectorFromFloat(Value=&FloatFromCurve(
      FloatCurve=DataInterface() {propertyValues {Curve {
         defaultValue=1
         keys [
          {arriveTangent=17.7998, interpMode=RCIM_Cubic, leaveTangent=17.7998, tangentMode=RCTM_User, time=0, value=0}
          {arriveTangentWeight=0.0701, interpMode=RCIM_Cubic, leaveTangentWeight=0.3438, tangentMode=RCTM_User, time=0.1328, value=1}
          {arriveTangentWeight=0.3276, interpMode=RCIM_Cubic, leaveTangentWeight=0.1084, tangentMode=RCTM_User, time=0.5914, value=1}
          {interpMode=RCIM_Cubic, leaveTangent=-3.9248, time=1, value=0}
         ]
        }}}
      Scale Curve=&MakeFloatFromVector(Vector=<1 1 1>, Channel=::NewEnumerator0)
     )))
   &Collision(
    Age Colliding Particles=false
    Collision Enabled=&IsPlatformSetActive(New Platform Set=DataInterface() {propertyValues {Platforms {qualityLevelMask=-4}}})
    Control Roll On Collision=true
    Enable Rest State=false
    Friction=0.7
    KillOnCollision=false
    Module.EnableMaxCollisionCount=false
    Module.Randomize Collision Normal Vector=false
    Particle Radius Scale=5
    Radius Calculation Type=::NewEnumerator1
    Radius Scale for Rotation=0
    Restitution=0
    ScaleRateOfRotation=true
   ) {module=Collision001}
   /SetVariables(Particles.MeshOrientation=&LerpQuaternion(
     Quaternion A:Quat4f=$Particles.Initial.MeshOrientation
     Quaternion B:Quat4f=<0 0 1 1>
     Lerp Factor:NiagaraFloat=&FloatFromCurve(FloatCurve=DataInterface() {propertyValues {Curve {
         defaultValue=0
         keys [{interpMode=RCIM_Cubic, tangentMode=RCTM_User, time=0.073, value=0}, {interpMode=RCIM_Cubic, tangentMode=RCTM_User, time=0.1946, value=1}]
        }}})
     Enable Spherical Linear Interpolation:NiagaraBool=false
    )) {enabled=false, isSetParameters=true}
   /Niagara/Modules/Collision/AlignParticlesWithCollisionPlane(Allow Alignment on Y=false, Allow Alignment on Z=false)
   &ScaleColor(
    Scale Alpha=&FloatFromCurve(FloatCurve=DataInterface() {propertyValues {Curve {keys [<at 0 1>,<at 0.75 1>,<at 1 0>]}}})
    Scale RGB=&Multiply_VectorByFloat(
     Vector=&MakeVectorFromLinearColorRGB(Color=<rgba 1 1 1 1>)
     Float=&MakeFloatFromVector(
      Vector=&VectorFromFloat(Value=&FloatFromCurve(FloatCurve=DataInterface() {propertyValues {Curve {keys [<at 0 2 RCIM_Cubic>,<at 0.25 1 RCIM_Cubic>]}}}))
      Channel=::NewEnumerator0
     )
    )
    ScaleA=false
    ScaleRGB=true
   )
   &SolveForcesAndVelocity(Clamp Velocity=false, Limit Acceleration=false)
  ]
  rendererClasses [&NiagaraMeshRendererProperties]
  renderers [{rendererClass=&NiagaraMeshRendererProperties, rendererIndex=0}]
 }
 FN_AnimTrail002 {
  emitterSpawnScript [&FN_RibbonIDOffsetUtility()]
  emitterUpdateScript [
   &EmitterState()
   &FN_SpawnAnimTrail(
    Number of Anim Trails:NiagaraInt32=&MakeIntFromFloat(Float=&MakeCustomFloatFromBool(
      Bool=$User.IsSkydiving
      True Float=0
      False Float=&MakeCustomFloatFromBool(
       Bool=$User.HideFootFX
       True Float=0
       False Float=&MakeCustomFloatFromBool(
        Bool=$System.isRunning
        True Float=&MakeCustomFloatFromBool(Bool=$System.isFalling, True Float=48, False Float=48)
        False Float=48
       )
      )
     ))
    Switch on ENiagara_AnimTrail_Source:ENiagara_AnimTrail_Source=::NewEnumerator1
   )
   /SetVariables(Emitter.Linear Color:LinearColor=&ColorFromCurve(
     ColorCurve=DataInterface() {propertyValues {
       AlphaCurve {keys []}
       BlueCurve {keys [<at 0 0.7475 RCIM_Cubic>,<at 1 0.283 RCIM_Cubic>]}
       GreenCurve {keys [<at 0 1.8026 RCIM_Cubic>,<at 1 2.3 RCIM_Cubic>]}
       RedCurve {keys [<at 0 4.5 RCIM_Cubic>,<at 1 4.5 RCIM_Cubic>]}
      }}
     CurveIndex=&Waveform("Waveform [1]"=::NewEnumerator0, "[1] Frequency"=0.5, "[1] Amplitude Min/Max"=<0 1>)
    )) {enabled=false, isSetParameters=true}
  ]
  particleSpawnScript [
   &FN_InitializeAnimTrailBoneIndecies(Second Bone or Socket Index=0)
   &FN_InitializeAnimTrailBoneIndecies(AnimTrail Selection=1, First Bone or Socket Index=2, Second Bone or Socket Index=2) {module=FN_InitializeAnimTrailBoneIndecies001}
   &FN_InitializeAnimTrail(
    AnimTrailWidth=&FN_FrontEndFloatModifier(Gameplay=30, Front End=25)
    Lifetime=&FN_FrontEndFloatModifier(Gameplay=0.125, Front End=0.135)
    SkeletalMesh=$System.Skeletal Mesh
   )
   &AddVelocity(Velocity=&MakeVector(
     X=0
     Y=&FN_FrontEndFloatModifier(Gameplay=0, Front End=-300)
     Z=&FloatFromCurve(
      FloatCurve=DataInterface() {propertyValues {Curve {keys [<at 0.0218 1>,<at 0.0388 0>]}}}
      Scale Curve=&FN_FrontEndFloatModifier(Gameplay=70, Front End=0)
     )
    ))
  ]
  particleUpdateScript [
   &SolveForcesAndVelocity(Clamp Velocity=false, Limit Acceleration=false)
   &FN_UpdateAnimTrail(SkeletalMesh=$System.Skeletal Mesh) {module=FN_UpdateAnimTrail001}
   &DynamicMaterialParameters(
    Index 0 Param 1=0.85
    Index 0 Param 2=1
    Index 0 Param 3=-0.5
    Index 0 Param 4=&MakeCustomFloatFromBool(
     Bool=&Boolean_AndOperation(Bool1=$System.bIsRunning, Bool2=&Boolean_NotOperation(Bool=$User.IsSkydiving))
     True Float=1
     False Float=0
    )
    Module.Index0Float0Write=true
    Module.Index0Float1Write=true
    Module.Index0Float2Write=true
    Module.Index0Float3Write=true
   )
   &ScaleRibbonWidth(Ribbon Width Scale=&FloatFromCurve(FloatCurve=DataInterface() {propertyValues {Curve {
        defaultValue=1
        keys [
         {interpMode=RCIM_Cubic, tangentMode=RCTM_User, time=0.0057, value=0}
         {arriveTangent=2.4408, interpMode=RCIM_Cubic, leaveTangent=2.4408, time=0.0846, value=0.671}
         {arriveTangent=1.1685, interpMode=RCIM_Cubic, leaveTangent=2.4675, tangentMode=RCTM_Break, time=0.304, value=0.7282}
         {arriveTangent=-3.3834, interpMode=RCIM_Cubic, leaveTangent=-3.3834, tangentMode=RCTM_User, time=0.8306, value=0.3545}
         {arriveTangent=-0.3484, tangentMode=RCTM_User, time=1, value=0}
        ]
       }}}))
   /SetVariables(Particles.Secondary Color:LinearColor=&ColorFromCurve(ColorCurve=DataInterface() {propertyValues {
       AlphaCurve {keys [<at 0 0.4 RCIM_Cubic>,<at 0.4247 0.2>,<at 1 0 RCIM_Cubic>]}
       BlueCurve {keys [
         <at 0 0.1015>
         {arriveTangent=0.5228, interpMode=RCIM_Cubic, leaveTangent=0.5228, time=0.0964, value=0.3963}
         <at 0.2834 0.2497>
         <at 0.5676 0.1914>
         <at 1 0 RCIM_Cubic>
        ]}
       GreenCurve {keys [
         <at 0 0.6079>
         {arriveTangent=3.6732, interpMode=RCIM_Cubic, leaveTangent=3.6732, time=0.0964, value=1.991}
         <at 0.2834 1.6489>
         <at 0.5676 0.9173>
         <at 1 0.6202 RCIM_Cubic>
        ]}
       RedCurve {keys [
         <at 0 1>
         {arriveTangent=3.5284, interpMode=RCIM_Cubic, leaveTangent=3.5284, time=0.0964, value=2.5}
         <at 0.2834 2>
         <at 0.5676 1.3>
         <at 1 1 RCIM_Cubic>
        ]}
      }})) {isSetParameters=true}
   /SetVariables(Particles.Update Secondary Color:LinearColor=&MultiplyLinearColor(
     Linear Color=$Particles.Secondary Color
     Scale Factor=&MakeVector4_FromLinearColor(LinearColor=&MakeLinearColorFromVectorAndFloat(
       "Vector (RGB)"=<1 1 1>
       "Float (Alpha)"=&MakeCustomFloatFromBool(
        Bool=&Boolean_AndOperation(Bool1=$System.bIsRunning, Bool2=&Boolean_NotOperation(Bool=$User.IsSkydiving))
        True Float=0.8
        False Float=0
       )
      ))
    )) {isSetParameters=true}
   /SetVariables(Particles.Linear Color=&MakeLinearColorFromVectorAndFloat(
     "Vector (RGB)"=&FN_FrontEndVector3Modifier(
      Gameplay=&MakeVectorFromLinearColorRGB(Color=&MultiplyLinearColor(
        Linear Color=&ColorFromCurve(ColorCurve=DataInterface() {propertyValues {
           AlphaCurve {keys []}
           BlueCurve {keys [<at 0.435 6.7156 RCIM_Cubic>,<at 0.6469 2.6061>,<at 0.8064 0.8275 RCIM_Cubic>]}
           GreenCurve {keys [<at 0.435 6.8951 RCIM_Cubic>,<at 0.6469 2.8042>,<at 0.8064 0.927 RCIM_Cubic>]}
           RedCurve {keys [<at 0.435 7 RCIM_Cubic>,<at 0.6469 3>,<at 0.8064 1 RCIM_Cubic>]}
          }})
        Scale Factor=&MakeVector4_FromLinearColor(LinearColor=&FNTimeOfDayColor(
          Night Color=<rgba 1 0.9432 0.8935 1>
          Morning Color=<rgba 1.15 1.1178 0.9341 1>
          Day Color=<rgba 1 1 1 1>
          Evening Color=<rgba 1.5 1.2664 0.7766 1>
         ))
       ))
      Front End=&MakeVectorFromLinearColorRGB(Color=&MultiplyLinearColor(
        Linear Color=<rgba 12 8.4991 0 1>
        Scale Factor=&MakeVector4_FromLinearColor(LinearColor=&FNTimeOfDayColor(
          Night Color=<rgba 1 0.9432 0.8935 1>
          Morning Color=<rgba 1.15 1.1178 0.9341 1>
          Day Color=<rgba 1 1 1 1>
          Evening Color=<rgba 1.5 1.2664 0.7766 1>
         ))
       ))
     )
     "Float (Alpha)"=&MakeCustomFloatFromBool(
      Bool=&Boolean_AndOperation(Bool1=$System.bIsRunning, Bool2=&Boolean_NotOperation(Bool=$User.IsSkydiving))
      True Float=&Multiply_Float(
       A=&FNTimeOfDayScalar(Night Scale=0.05, Morning Scale=0.08, Day Scale=0.04, Evening Scale=0.08)
       B=&FloatFromCurve(FloatCurve=DataInterface() {propertyValues {Curve {
           defaultValue=1
           keys [
            {tangentMode=RCTM_User, time=0, value=1}
            {arriveTangent=-0.0405, interpMode=RCIM_Cubic, leaveTangent=-0.0405, tangentMode=RCTM_User, time=0.3622, value=1}
            {arriveTangent=-1.568, interpMode=RCIM_Cubic, leaveTangent=-1.568, time=0.7065, value=0.3913}
            {arriveTangent=-0.0717, interpMode=RCIM_Cubic, leaveTangent=-0.0717, tangentMode=RCTM_User, time=1, value=0}
           ]
          }}})
      )
      False Float=0
     )
    )) {isSetParameters=true}
  ]
  rendererClasses [&NiagaraRibbonRendererProperties]
  renderers [
   {rendererClass=&NiagaraRibbonRendererProperties}
   {rendererClass=&NiagaraRibbonRendererProperties, rendererIndex=1}
  ]
 }
 Fountain {
  emitterUpdateScript [
   &EmitterState()
   &SpawnRate(
    SpawnRate=&MakeCustomFloatFromBool(
     Bool=&Boolean_AndOperation(Bool1=$System.bIsRunning, Bool2=&Boolean_NotOperation(Bool=$User.IsSkydiving))
     True Float=35
     False Float=0
    )
    Use Spawn Probability=false
   )
  ]
  particleSpawnScript [
   &PartitionParticles()
   &InitializeParticle(
    Color Maximum=<rgba 1 0.8754 0.6506 0.767>
    Color Minimum=<rgba 0.6906 0.5742 0.2826 0.7183>
    Color Mode=::NewEnumerator2
    Lifetime Max=0.2
    Lifetime Min=0.05
    Lifetime Mode=::NewEnumerator1
    Mass Mode=::NewEnumerator3
    Mesh Scale Mode=::NewEnumerator3
    Sprite Rotation Mode=::NewEnumerator2
    Sprite Size Mode=::NewEnumerator3
    Uniform Sprite Size=&MaskFloat(
     Mask Source=0
     Mask Parameter=?
     Mask=0
     Masked Float=&RandomRangeFloat(Minimum=7, Maximum=14)
     Passthrough Float=&Multiply_Float(A=0.01, B=&RandomRangeFloat(Minimum=0.01, Maximum=0.02))
    )
    UsePositionOffset=false
    Write Material Random=true
   )
   &SkeletalMeshLocation(
    Bone / Socket Index=$Particles.PartitionParticles.Partition
    MaskSpawnGroup=false
    Mesh Sampling Type=::NewEnumerator5
    Skeletal Mesh=$System.Skeletal Mesh
    Socket Sampling Mode=::NewEnumerator5
   )
   &ShapeLocation(MaskSpawnGroup=false, Override Seed=false, Sphere Radius=10, UseOwnerScale=true)
   &AddVelocity(Module.Use Falloff Distance=false, Origin Offset=<0 2 0>, Override Seed=false, Velocity Mode=::NewEnumerator1, Velocity Speed=15)
   &AddVelocity(Velocity=<0 -20 5>) {module=AddVelocity001}
  ]
  particleUpdateScript [
   &ParticleState()
   &GravityForce(Gravity=<0 0 -300>)
   &Drag(Drag=0.25, Use Linear Drag=true, Use Rotational Drag=false)
   &ScaleColor(
    Scale Alpha=&FloatFromCurve(FloatCurve=DataInterface() {propertyValues {Curve {keys [<at -0.0046 1.0113 RCIM_Cubic>,<at 1.0011 0.002 RCIM_Cubic>]}}})
    ScaleA=true
    ScaleRGB=true
   )
   &InheritSourceMovement(Applied Position Scale=<0.5 0.5 0.5>, ApplyForce=false, ApplyPosition=true, ApplyPositionFalloff=false)
   &SolveForcesAndVelocity(Clamp Velocity=false, Limit Acceleration=false)
   &ScaleSpriteSizeBySpeed(Max Scale Factor=<1 5>, Min Scale Factor=<0.5 1>) {enabled=false}
   &ScaleSpriteSize(Scale Factor=&Vector2DFromFloat(Value=&FloatFromCurve(
      FloatCurve=DataInterface() {propertyValues {Curve {
         defaultValue=0.5
         keys [
          {arriveTangent=6.4847, interpMode=RCIM_Cubic, leaveTangent=6.4847, tangentMode=RCTM_User, time=0.0027, value=0}
          {interpMode=RCIM_Cubic, tangentMode=RCTM_User, time=0.2916, value=1.041}
          {tangentMode=RCTM_User, time=1, value=0}
         ]
        }}}
      Scale Curve=&MaskFloat(
       Mask Source=0
       Mask Parameter=?
       Mask=0
       Masked Float=1
       Passthrough Float=&RandomRangeFloat(Minimum=0.3, Maximum=0.5)
      )
     )))
  ]
  rendererClasses [&NiagaraSpriteRendererProperties]
  renderers [{rendererClass=&NiagaraSpriteRendererProperties}]
 }
}
summary {
 systemName=NS_Shoe_JoyGridForge_Trail
 emitters [
  {emitterName=Fountain, rendererClasses [&NiagaraSpriteRendererProperties]}
  {emitterName=Bananas, rendererClasses [&NiagaraMeshRendererProperties]}
  {emitterName=FN_AnimTrail002, rendererClasses [&NiagaraRibbonRendererProperties]}
 ]
 userVariables [
  User.bisFrontEnd {type=NiagaraBool}
  User.bIsFrontEndPreview {type=NiagaraBool}
  User.bisPrimaryPreviewItem {type=NiagaraBool}
  User.HideFootFX {type=NiagaraBool}
  User.IsSkydiving {type=NiagaraBool}
  User.RibbonIDOffset {type=NiagaraInt32}
 ]
}
